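# Animation Blueprint for the FinchVisit pickaxe cosmetic, adding procedural secondary motion to its wire and face bones.
# The input pose runs through four chained Anim Dynamics nodes, simulated relative to the two-handed handle bone with overridden gravity.
# Both wire bones and the first face bone use limited angular springs, while the second face bone slides on a single-axis linear spring.
# The event graph holds only unconnected default nodes.
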AnimBlueprint /BRCosmetics/Weapons/FORT_Melee/Pickaxe_FinchVisit/Meshes/FinchVisit_Axe_AnimBP
target_skeleton=&/FinchVisit_Axe_Skeleton
graphs [
 AnimGraph {nodes [
   AnimGraphNode_SubInput_0 = Input Pose()
   n1 = Anim Dynamics - Bone: dyn_wires_01(ComponentPose=Local To Component(LocalPose=@AnimGraphNode_SubInput_0), Alpha=1, ExternalForce="0,0,0") {
    angularDampingOverride=1
    angularSpringConstant=10000
    bAngularSpring=true
    bGravityOverrideInSimSpace=true
    bOverrideAngularDamping=true
    bUseGravityOverride=true
    simulationSpace=BoneRelative
    alphaBoolBlend {blendOption=Linear}
    boundBone {boneName=dyn_wires_01}
    componentAppliedLinearAccClamp=<3000 3000 3000>
    gravityOverride {z=-980}
    physicsBodyDefinitions [{
      boundBone {boneName=dyn_wires_01}
      boxExtents=<5 2.1024 5>
      constraintSetup {angularLimitsMax=<0 10 20>, angularLimitsMin=<0 -5 -20>, angularTarget=<0 0 0>}
      localJointOffset=<2 0 0>
     }]
    relativeSpaceBone {boneName=Melee_TwoHanded_Handle}
    simSpaceSettings {maxAngularAcceleration=7, maxAngularVelocity=1.5, simSpaceAngularAlpha=1}
   }
   n2 = Anim Dynamics - Bone: dyn_wires_02(ComponentPose=@n1, Alpha=1, ExternalForce="0,0,0") {
    angularDampingOverride=0.999
    angularSpringConstant=1000
    bAngularSpring=true
    bGravityOverrideInSimSpace=true
    bOverrideAngularDamping=true
    bUseGravityOverride=true
    simulationSpace=BoneRelative
    alphaBoolBlend {blendOption=Linear}
    boundBone {boneName=dyn_wires_02}
    componentAppliedLinearAccClamp=<2000 2000 2000>
    gravityOverride {z=-500}
    physicsBodyDefinitions [{
      boundBone {boneName=dyn_wires_02}
      boxExtents=<4.9189 9.9765 4.8993>
      constraintSetup {angularLimitsMax=<0 10 5>, angularLimitsMin=<0 -5 -5>, angularTarget=<0 0 0>}
      localJointOffset=<1 0 0>
     }]
    relativeSpaceBone {boneName=Melee_TwoHanded_Handle}
    simSpaceSettings {maxAngularAcceleration=7, maxAngularVelocity=1.5, simSpaceAngularAlpha=1}
   }
   n3 = Anim Dynamics - Bone: dyn_face_01(ComponentPose=@n2, Alpha=1, ExternalForce="0,0,0") {
    angularDampingOverride=1
    angularSpringConstant=1000
    bAngularSpring=true
    bGravityOverrideInSimSpace=true
    bOverrideAngularDamping=true
    bUseGravityOverride=true
    bUsePlanarLimit=false
    simulationSpace=BoneRelative
    alphaBoolBlend {blendOption=Linear}
    boundBone {boneName=dyn_face_01}
    componentAppliedLinearAccClamp=<300 300 300>
    gravityOverride {z=-500}
    physicsBodyDefinitions [{
      boundBone {boneName=dyn_face_01}
      boxExtents=<15 15 5>
      constraintSetup {angularLimitsMax=<0 10 30>, angularLimitsMin=<0 -10 -3>, angularTarget=<0 0 0>}
      localJointOffset=<1 0 0>
     }]
    relativeSpaceBone {boneName=Melee_TwoHanded_Handle}
    simSpaceSettings {maxAngularAcceleration=7, maxAngularVelocity=1.5, simSpaceAngularAlpha=1}
   }
   n4 = Anim Dynamics - Bone: dyn_face_02(ComponentPose=@n3, Alpha=1, ExternalForce="0,0,0") {
    angularDampingOverride=0.999
    angularSpringConstant=50
    bGravityOverrideInSimSpace=true
    bLinearSpring=true
    bOverrideLinearDamping=true
    bUseGravityOverride=true
    linearDampingOverride=1
    linearSpringConstant=150
    simulationSpace=BoneRelative
    alphaBoolBlend {blendOption=Linear}
    boundBone {boneName=dyn_face_02}
    componentAppliedLinearAccClamp=<500 500 500>
    gravityOverride {x=500}
    physicsBodyDefinitions [{
      boundBone {boneName=dyn_face_02}
      boxExtents=<20 20 20>
      constraintSetup {angularLimitsMax=<0 0 0>, angularLimitsMin=<0 0 0>, angularTarget=<0 0 0>, linearAxesMax=<10 0 0>, linearAxesMin=<-1 0 0>}
      localJointOffset=<-0.5 0 0>
     }]
    relativeSpaceBone {boneName=Melee_TwoHanded_Handle}
    simSpaceSettings {maxAngularAcceleration=7, maxAngularVelocity=1.5, simSpaceAngularAlpha=1}
   }
   Output Pose(Result=Component To Local(ComponentPose=@n4)) {layerGroup=None}
  ]}
 EventGraph {nodes [
   TryGetPawnOwner()
   Event BlueprintUpdateAnimation()
  ]}
]
variables []
